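# Idle Niagara effect for the ScribeSpiderOar cosmetic shoes, adding lightning to the wearer's skeletal mesh.
# LightningTrails draws four short bone-driven ribbon trails, with velocity and lifetime adjusted for front end versus gameplay.
# MergedSplineLightning spawns brief socket-attached mesh bolts, animated through randomized, curve-driven dynamic material parameters.
# System.ShowFX gates both emitters, System.TOD tints their colors, and per-foot ShoeScale checks kill particles at spawn.
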NiagaraSystem /CosmeticShoes/Assets/FORT_Shoes/Shoe_ScribeSpiderOar/FX/NS_Shoe_ScribeSpiderOar_Idle
dependencies {
 usedDataInterfaces [
  &NiagaraDataInterfaceCurve
  &NiagaraDataInterfaceColorCurve
  &NiagaraDataInterfaceMeshRendererInfo
  &NiagaraDataInterfaceSkeletalMesh
 ]
 usedDynamicInputs [
  &Boolean_AndOperation
  &Boolean_NotOperation
  &VectorLength
  &Multiply_Vector
 ]
 usedModules [
  &self:SystemSpawnScript.SystemScriptSource.SystemScriptGraph.NiagaraNodeAssignment_0.SetVariables_#1
  &self:SystemSpawnScript.SystemScriptSource.SystemScriptGraph.NiagaraNodeAssignment_5.SetVariables_#2
  &SystemState
  &self:SystemSpawnScript.SystemScriptSource.SystemScriptGraph.NiagaraNodeAssignment_4.SetVariables_#3
  &self:SystemSpawnScript.SystemScriptSource.SystemScriptGraph.NiagaraNodeAssignment_1.SetVariables_#4
  &Timeline
  &self:SystemSpawnScript.SystemScriptSource.SystemScriptGraph.NiagaraNodeAssignment_3.SetVariables_#5
 ]
}
emitters {
 LightningTrails {
  emitterSpawnScript [&FN_RibbonIDOffsetUtility()]
  emitterUpdateScript [
   &EmitterState()
   &FN_SpawnAnimTrail(
    Number of Anim Trails:NiagaraInt32=&MakeCustomIntFromBool(Bool=$System.ShowFX, True INT=4, False INT=0)
    Switch on ENiagara_AnimTrail_Source:ENiagara_AnimTrail_Source=::NewEnumerator1
   )
  ]
  particleSpawnScript [
   &FN_InitializeAnimTrailBoneIndecies(First Bone or Socket Index=2, Second Bone or Socket Index=2)
   &FN_InitializeAnimTrailBoneIndecies(AnimTrail Selection=1, First Bone or Socket Index=4, Second Bone or Socket Index=4) {module=FN_InitializeAnimTrailBoneIndecies001}
   &FN_InitializeAnimTrailBoneIndecies(AnimTrail Selection=2, First Bone or Socket Index=5, Second Bone or Socket Index=5) {module=FN_InitializeAnimTrailBoneIndecies003}
   &FN_InitializeAnimTrailBoneIndecies(AnimTrail Selection=3, First Bone or Socket Index=7, Second Bone or Socket Index=7) {module=FN_InitializeAnimTrailBoneIndecies004}
   &FN_InitializeAnimTrail(
    AnimTrailWidth=&MergeHelper_FloatSwitch(
     0=20
     1=15
     2=10
     3=?
     Index=/Niagara/DynamicInputs/Divide/Divide_Int(A:NiagaraInt32=$Particles.SpawnGroup, B:NiagaraInt32=2)
     Num=::NewEnumerator1
    )
    Color=&MakeLinearColorFromVectorAndFloat(
     "Vector (RGB)"=&MakeVectorFromLinearColorRGB(Color=&MultiplyLinearColor(
       Linear Color=<rgba 0.9583 0.8086 0.8101 1>
       Scale Factor=&MakeVector4_FromLinearColor(LinearColor=$System.TOD)
      ))
     "Float (Alpha)"=&Add_Float(A=$System.Timeline001.Time, B=1)
    )
    Lifetime=&FN_FrontEndFloatModifier(Gameplay=0.08, Front End=0.07)
    SkeletalMesh=$System.Skeletal Mesh
   )
   &AddVelocity(
    Velocity=&RandomRangeVector(Minimum=<-10 -10 -10>, Maximum=<10 10 10>)
    Velocity Speed Scale=&Multiply_Float(
     A=5
     B=&MergeHelper_FloatSwitch(
      0=1
      1=1.5
      2=2
      3=?
      Index=&Divide_Int(A:NiagaraInt32=$Particles.SpawnGroup, B:NiagaraInt32=2)
      Num=::NewEnumerator1
     )
    )
   )
   &AddVelocity(Velocity=&FN_FrontEndVector3Modifier(Gameplay=<0 0 0>, Front End=<160 -500 200>)) {module=AddVelocity001}
   &KillParticles(Kill Particles=&SetBoolByFloatComparison(
     A=&MakeFloatFromVector(
      Vector=&MergeHelper_VectorSwitch(
       0=$User.ShoeScale_L
       1=$User.ShoeScale_L
       2=$User.ShoeScale_R
       3=$User.ShoeScale_R
       4=$User.ShoeScale_R
       5=$User.ShoeScale_R
       Index=&SpawnGroup()
       Num=::NewEnumerator2
      )
      Channel=::NewEnumerator0
     )
     B=0.05
     Comparison Type=::NewEnumerator6
    ))
  ]
  particleUpdateScript [
   /SetVariables(Particles.InvNormalizedAge:NiagaraFloat=&Divide_Float(A=&OneMinusFloat(Float=$Particles.NormalizedAge), B=1)) {isSetParameters=true}
   &FN_UpdateAnimTrail(SkeletalMesh=$System.Skeletal Mesh) {module=FN_UpdateAnimTrail001}
   &ScaleRibbonWidth(Ribbon Width Scale=&FloatFromCurve(FloatCurve=DataInterface() {propertyValues {Curve {keys [{arriveTangent=-0.03, interpMode=RCIM_Cubic, leaveTangent=-0.03, tangentMode=RCTM_User, time=0.1, value=1}, <at 1 0 RCIM_Cubic>]}}})) {module=ScaleRibbonWidth001}
   &ScaleColor(
    Scale Alpha=&FloatFromCurve(FloatCurve=DataInterface() {propertyValues {Curve {keys [<at 0.5 1>,<at 0.8 0.5>]}}})
    Scale RGB=&MakeVectorFromLinearColorRGB(Color=&ColorFromCurve(ColorCurve=DataInterface() {propertyValues {
        AlphaCurve {keys [<at 0 1 RCIM_Cubic>]}
        BlueCurve {keys [<at 0 1 RCIM_Cubic>,<at 1 0.6615 RCIM_Cubic>]}
        GreenCurve {keys [<at 0 1 RCIM_Cubic>,<at 1 0.6041 RCIM_Cubic>]}
        RedCurve {keys [<at 0 1 RCIM_Cubic>,<at 1 0.3903 RCIM_Cubic>]}
       }}))
    ScaleA=true
    ScaleRGB=true
   )
   &SolveForcesAndVelocity(Clamp Velocity=false, Limit Acceleration=false)
   &DynamicMaterialParameters(
    Index 1 Param 3=-1
    Index 2 Param 3=&MergeHelper_FloatSwitch(0=0.5, 1=1.5, 2=2.5, 3=3.5, Index=$Particles.SpawnGroup, Num=::NewEnumerator2)
    Module.Index1Float0Write=true
    Module.Index1Float1Write=true
    Module.Index1Float2Write=true
    Module.Index1Float3Write=true
    Module.Index2Float0Write=true
    Module.Index2Float1Write=true
    Module.Index2Float2Write=true
    Module.Index2Float3Write=true
    Write Parameter Index 0=::NewEnumerator0
    Write Parameter Index 1=::NewEnumerator1
    Write Parameter Index 2=::NewEnumerator1
   )
  ]
  rendererClasses [&NiagaraRibbonRendererProperties]
  renderers [
   {rendererClass=&NiagaraRibbonRendererProperties}
   {rendererClass=&NiagaraRibbonRendererProperties, rendererIndex=1}
  ]
 }
 MergedSplineLightning {
  emitterUpdateScript [
   &EmitterState(
    Inactive Response=::NewEnumerator1
    Life Cycle Mode=::NewEnumerator1
    Loop Behavior=::NewEnumerator1
    Loop Duration Mode=::NewEnumerator1
    UseLoopDelay=false
   )
   &SpawnRate(
    Spawn Probability=&MakeCustomFloatFromBool(Bool=$System.ShowFX, True Float=1, False Float=0)
    SpawnRate=30
    Use Spawn Probability=true
   )
  ]
  particleSpawnScript [
   /SetVariables(Particles.VisibilityTag=&SpawnGroup()) {isSetParameters=true}
   &PartitionParticles()
   /SetVariables(Particles.Random:NiagaraInt32=&RandomRangeInt(Minimum=0, Maximum=1)) {isSetParameters=true}
   &InitializeParticle(
    Color=&MultiplyLinearColor(Linear Color=<rgba 1 0.7728 0.276 1>, Scale Factor=&MakeVector4_FromLinearColor(LinearColor=$System.TOD))
    Lifetime=&RandomRangeFloat(Minimum=0.25, Maximum=0.35)
    Mesh Scale Max=<0.8 0.8 0.8>
    Mesh Scale Min=<0.7 0.7 0.7>
    Mesh Scale Mode=::NewEnumerator2
    UsePositionOffset=false
    Write Material Random=true
   ) {module=InitializeParticle001}
   &CameraOffset(Camera Offset Amount=5)
   /SetVariables(Particles.LightningIndex:NiagaraInt32=&RandomRangeInt(Minimum=0, Maximum=5)) {isSetParameters=true}
   &InitialMeshOrientation(
    Mesh Orientation Mode=::NewEnumerator7
    Rotation=&RandomRangeVector(Minimum=<-0.1 -0.1 -1>, Maximum=<0.1 0.1 1>)
    Use Rotation Vector=true
   )
   &KillParticles(Kill Particles=&SetBoolByFloatComparison(
     A=&MakeFloatFromVector(
      Vector=&MergeHelper_VectorSwitch(0=$User.ShoeScale_L, 1=$User.ShoeScale_R, 2=$User.ShoeScale_L, 3=?, Index=$Particles.PartitionParticles.Partition, Num=::NewEnumerator0)
      Channel=::NewEnumerator0
     )
     B=0.05
     Comparison Type=::NewEnumerator6
    ))
  ]
  particleUpdateScript [
   &ParticleState()
   &SkeletalMeshLocation(
    Bone / Socket Index=$Particles.PartitionParticles.Partition
    MaskSpawnGroup=false
    Mesh Sampling Type=::NewEnumerator5
    Orientation Sampling=::NewEnumerator1
    Skeletal Mesh=$System.Skeletal Mesh
    Socket Sampling Mode=::NewEnumerator5
   )
   &DynamicMaterialParameters(
    Index 0 Param 1=&Multiply_Float(
     A=&RandomRangeFloat(Minimum=3, Maximum=5)
     B=&FloatFromCurve(FloatCurve=DataInterface() {propertyValues {Curve {keys [{interpMode=RCIM_Cubic, leaveTangent=8.1085, tangentMode=RCTM_Break, time=0.8568, value=1}, {tangentMode=RCTM_User, time=1, value=0}]}}})
    )
    Index 0 Param 2=&Divide_Float(A=&RandomRangeFloat(Minimum=1, Maximum=1.5), B=1.5)
    Index 0 Param 3=&Lerp_Float(
     A=&RandomRangeFloat(Minimum=0, Maximum=0.3)
     B=1
     Alpha=&FloatFromCurve(FloatCurve=DataInterface() {propertyValues {Curve {keys [<at 0 1>,<at 0.1 0.375>,<at 0.5 0.05>,<at 1 0>]}}})
    )
    Index 0 Param 4=&RandomRangeFloat(Minimum=0, Maximum=1)
    Index 1 Param 1=1
    Index 1 Param 2=1
    Index 1 Param 3=1
    Index 2 Param 1=&Multiply_Float(
     A=&FloatFromCurve(FloatCurve=DataInterface() {propertyValues {Curve {keys [<at 0 0.5>,<at 1 1>]}}})
     B=&RandomRangeFloat(Minimum=3, Maximum=4)
    )
    Index 2 Param 2=&FloatFromCurve(
     FloatCurve=DataInterface() {propertyValues {Curve {keys [{interpMode=RCIM_Cubic, tangentMode=RCTM_Break, time=0.1, value=1}, <at 1 0>]}}}
     Scale Curve=&RandomRangeFloat(Minimum=1, Maximum=1)
    )
    Module.Index0Param0Write=true
    Module.Index0Param1Write=true
    Module.Index0Param2Write=true
    Module.Index0Param3Write=true
    Module.Index1Param0Write=true
    Module.Index1Param1Write=true
    Module.Index1Param2Write=true
    Module.Index1Param3Write=true
    Module.Index2Param0Write=true
    Module.Index2Param1Write=true
    Module.Index2Param2Write=true
    Module.Index2Param3Write=true
    Write Parameter Index 1=true
    Write Parameter Index 2=true
   ) {module=DynamicMaterialParameters001}
   &ScaleColor(
    Scale Alpha=&FloatFromCurve(FloatCurve=DataInterface() {propertyValues {Curve {keys [
         {tangentMode=RCTM_User, time=0, value=1}
         {arriveTangent=-0.0405, interpMode=RCIM_Cubic, leaveTangent=-0.0405, tangentMode=RCTM_User, time=0.3, value=1}
         {interpMode=RCIM_Cubic, leaveTangent=-8.0089, time=1, value=0}
        ]}}})
    Scale RGB=&Lerp_Vector(
     A=&VectorFromFloat(Value=10)
     B=&VectorFromFloat(Value=&RandomRangeFloat(Minimum=0.8, Maximum=1, Evaluation Type=::NewEnumerator1))
     Alpha=&FloatFromCurve(FloatCurve=DataInterface() {propertyValues {Curve {keys [
          {interpMode=RCIM_Cubic, leaveTangent=2.5294, tangentMode=RCTM_Break, time=0.1, value=0}
          {arriveTangent=-0.2166, interpMode=RCIM_Cubic, tangentMode=RCTM_Break, time=0.4, value=1}
         ]}}})
    )
    ScaleA=true
    ScaleRGB=true
   ) {module=ScaleColor001}
   /SetVariables(Particles.GlowColor=&MultiplyLinearColor(
     Linear Color=<rgba 3.3076 5 2.416 1>
     Scale Factor=&MakeVector4FromVectorAndFloat(
      "Vector (RGB)"=<1 1 1>
      "Float (Alpha)"=&Multiply_Float(
       A=0.2
       B=&FloatFromCurve(FloatCurve=DataInterface() {propertyValues {Curve {keys [
            {tangentMode=RCTM_User, time=0, value=1}
            {arriveTangent=-1, interpMode=RCIM_Cubic, leaveTangent=-1, time=0.2, value=1}
            {interpMode=RCIM_Cubic, leaveTangent=-8.0089, time=1, value=0}
           ]}}})
      )
     )
    )) {isSetParameters=true}
  ]
  rendererClasses [&NiagaraMeshRendererProperties]
  renderers [
   {rendererClass=&NiagaraMeshRendererProperties, rendererIndex=0}
   {rendererClass=&NiagaraMeshRendererProperties, rendererIndex=1}
  ]
 }
}
summary {
 systemName=NS_Shoe_ScribeSpiderOar_Idle
 emitters [
  {emitterName=LightningTrails, rendererClasses [&NiagaraRibbonRendererProperties]}
  {emitterName=MergedSplineLightning, rendererClasses [&NiagaraMeshRendererProperties]}
 ]
 userVariables [
  User.bIsFrontEnd {type=NiagaraBool}
  User.bIsFrontEndPreview {type=NiagaraBool}
  User.bisLit {type=NiagaraBool}
  User.bisPrimaryPreviewItem {type=NiagaraBool}
  User.HideFootFX {type=NiagaraBool}
  User.IsSkydiving {type=NiagaraBool}
  User.RibbonIDOffset {type=NiagaraInt32}
  User.ShoeScale_L {type=Vector3f, description="", defaultValue {struct=Vector3f, value=<1 1 1>}}
  User.ShoeScale_R {type=Vector3f, description="", defaultValue {struct=Vector3f, value=<1 1 1>}}
 ]
}
